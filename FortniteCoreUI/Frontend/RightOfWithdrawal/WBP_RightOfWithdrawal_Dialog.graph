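# Activatable dialog that asks the player to accept losing the 14-day right of withdrawal for immediate delivery of purchased digital content.
# On activation it logs a sub-view analytics event, sets the scroll widget, binds the UIKit dialog's button dispatcher and pushes placeholder body text into its viewmodel.
# Accept or Cancel logs the interaction, broadcasts OnDialogDismissed with the result and deactivates the widget, which unbinds the dispatcher.
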WidgetBlueprint /FortniteCoreUI/Frontend/RightOfWithdrawal/WBP_RightOfWithdrawal_Dialog
graphs [
 EventGraph {nodes [
   n1 = OnButtonsClicked!() -> (OutputDelegate, then, ButtonIndex:int) {is=title K2Node_CustomEvent}
   > n2 = Switch on Int!(execute:Exec, Selection:int=@n1.ButtonIndex) -> (Default:Exec, 0:Exec, 1:Exec)
   n3 = Self-Reference()
   FireEventElementInteractedByName!(execute=@n2.0, Element Name:String=Button_Accept, WidgetCallingFrom:Widget Object Reference=@n3) -> (then, Success:bool)
   > n4 = Call On Dialog Dismissed!(execute:Exec, self:WBP Right Of Withdrawal Dialog Object Reference, Accepted:bool=true) -> (then:Exec) {is=title K2Node_CallDelegate}
   DeactivateWidget(execute=[@n5.then, @n4.then])
   FireEventElementInteractedByName(execute=@n2.1, Element Name=Button_Cancel, WidgetCallingFrom=@n3)
   > n5 = Call On Dialog Dismissed()
   Event On Activated()
   > Fire Event Sub View Opened!(execute, View:Widget Object Reference=Self-Reference(), AdditionalInfo:Flow Analytics Additional Info Structure, bIsOverlay:bool=true) -> (then)
   > SetScrollWidget!(
    execute
    ToScroll:Widget Object Reference=Get ScrollBox_Text!(self:WBP UIKit Dialog Internal Object Reference=Get Dialog_Internal() -> (Dialog_Internal:WBP UIKit Dialog Internal Object Reference)) -> (ScrollBox_Text:Scroll Box Object Reference)
   ) -> (then)
   > Bind Event to Dialog Button Clicked Dispatcher(self:WBP UIKit Dialog Internal Object Reference=Get Dialog_Internal(), Delegate=@n1) {is=title K2Node_AddDelegate}
   > Set DialogInfoText!(
    execute
    DialogInfoText:Text=Get BodyText() -> (BodyText:Text)
    self:UI Kit Dialog Viewmodel Object Reference=Get DialogVM_RightOfWithdrawal() -> (DialogVM_RightOfWithdrawal:Dialog VM Right Of Withdrawal Object Reference)
   ) -> (then, Output_Get:Text)
   Event On Deactivated()
   > Unbind Event from Dialog Button Clicked Dispatcher(self:WBP UIKit Dialog Internal Object Reference=Get Dialog_Internal(), Delegate=@n1) {is=title K2Node_RemoveDelegate}
  ]}
 OnDialogDismissed {nodes [OnDialogDismissed!() -> (then, Accepted:bool=~) {is=title K2Node_FunctionEntry}]}
]
variables [
 BodyText {
  type=string
  default="[PH] By clicking “Accept”, you agree to the immediate delivery of digital content you are purchasing. By doing so, you lose your 14 day cancellation right granted by law.  <text color=\"FFFFFFFF\" fontface=\"75 Medium\">You can still cancel within 24 hours if you haven't used the item in-game, or return the digital content with a Return Ticket, if you have any.</>"
 }
 OnDialogDismissed {}
]
widgets {
 info {
  inheritedWidgetCount=0
  parentClass=&CommonActivatableWidget
  rootWidgetClass=/FortUILibrary/Systems/Dialog/WBP_UIKit_Dialog_Internal.WBP_UIKit_Dialog_Internal_C
  widgetCount=1
 }
 widgets [Dialog_Internal {
   widgetClassPath=/FortUILibrary/Systems/Dialog/WBP_UIKit_Dialog_Internal
   bIsVariable=true
   widget=Dialog_Internal
   properties {
    designTimePreview_DialogVM=&/DialogVM_RightOfWithdrawal.DialogVM_RightOfWithdrawal_C
    useText=Common Rich Text
    backgroundProperties {
     material=/FortUILibrary/UIKit/Dialogs/MaterialInstances/MI_UI_DialogBackground
     image Size=<32 32>
     placement {
      horizontal Alignment=HAlign_Fill
      vertical Alignment=VAlign_Fill
      grid {column=0, column Span=2, layer=1, row=0, row Span=0, nudge=<0 0>}
      padding=<ltrb 0 0 0 0>
     }
    }
    buttons {
     each Button Horizontal Fill=false
     each Button Min Width=200
     each Button Size=Medium
     focuse Button=0
     space Between Buttons=30
     cTA Buttons Types [Regular,Quiet]
     group Placement {
      horizontal Alignment=HAlign_Fill
      vertical Alignment=VAlign_Bottom
      grid {column=1, column Span=0, layer=0, row=2, row Span=0, nudge=<0 0>}
      padding=<ltrb 80 40 80 48>
     }
    }
    dialogSIze {
     height Override {enable=false, value=540}
     max Aspect Ratio {enable=false, value=0}
     max Desired Height {enable=true, value=800}
     max Desired Width {enable=false, value=0}
     min Aspect Ratio {enable=false, value=0}
     min Desired Height {enable=false, value=0}
     min Desired Width {enable=false, value=0}
     width Override {enable=true, value=932}
    }
    imageProperties {
     material=/FortUILibrary/Systems/Dialog/MaterialInstances/MI_UI_DialogImage
     image Size=<512 512>
     placement {
      horizontal Alignment=HAlign_Fill
      vertical Alignment=VAlign_Fill
      grid {column=1, column Span=0, layer=20, row=0, row Span=1, nudge=<0 0>}
      padding=<ltrb 0 0 0 0>
     }
    }
    textProperties {
     auto Wrap Text=true
     justification=Left
     line Height Percentage=1.5
     mobile Font Size Multiplier=1.3
     transform Policy=None
     wrap Text At=0
     color {colorUseRule=UseColor_Specified, specifiedColor=<rgba 1 1 1 1>}
     font {
      bForceMonospaced=false
      bMaterialIsStencil=false
      fontMaterial=null
      fontObject=&HeadingNow
      letterSpacing=0
      monospacedWidth=1
      size=15.75
      skewAmount=0
      typefaceFontName=74 Regular
      outlineSettings {
       bApplyOutlineToDropShadows=false
       bMiteredCorners=false
       bSeparateFillAlpha=false
       outlineMaterial=null
       outlineSize=0
       outlineColor=<rgba 0 0 0 1>
      }
     }
     placement {
      horizontal Alignment=HAlign_Fill
      vertical Alignment=VAlign_Top
      grid {column=1, column Span=0, layer=0, row=1, row Span=0, nudge=<0 0>}
      padding=<ltrb 80 0 80 0>
     }
    }
    titleProperties {
     auto Wrap Text=true
     justification=Left
     line Height Percentage=1
     mobile Font Size Multiplier=1.3
     transform Policy=ToUpper
     wrap Text At=0
     color {colorUseRule=UseColor_Specified, specifiedColor=<rgba 1 1 1 1>}
     font {
      bForceMonospaced=false
      bMaterialIsStencil=false
      fontMaterial=null
      fontObject=&HeadingNow
      letterSpacing=0
      monospacedWidth=1
      size=25.5
      skewAmount=0
      typefaceFontName=86 Bold Italic
      outlineSettings {
       bApplyOutlineToDropShadows=false
       bMiteredCorners=false
       bSeparateFillAlpha=false
       outlineMaterial=null
       outlineSize=0
       outlineColor=<rgba 0 0 0 1>
      }
     }
     placement {
      horizontal Alignment=HAlign_Fill
      vertical Alignment=VAlign_Top
      grid {column=1, column Span=0, layer=0, row=0, row Span=0, nudge=<0 0>}
      padding=<ltrb 80 56 80 32>
     }
    }
   }
  }]
}
